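# Physics asset for the female medium WavyMan player body, covering three dynamic valve bones on the arms and one leg.
# Capsule bodies sit on both upper arms, the left calf and the three valve bones.
# Each valve is constrained to its limb, left and right valves to the matching upper arms and the leg valve to the left calf, with narrow swing and tighter twist limits.
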PhysicsAsset /Game/Characters/Player/Female/Medium/Bodies/F_MED_WavyMan/Meshes/F_MED_Wavy_Man_Physics
bodies {
 calf_l [{shapeType=Capsule, length=10.6154, radius=8.1431, center=<-11.8394 0 0>, rotation=<rot 90 0 0>}]
 dyn_valve_l [{shapeType=Capsule, length=2.7381, radius=1.6864, center=<1.4507 -0.0001 0>, rotation=<rot 90 0 0>}]
 dyn_valve_leg [{shapeType=Capsule, length=3.5495, radius=2.0349, center=<1.9028 -0.1672 -0.0794>, rotation=<rot -1.1278 -89.9993 87.5844>}]
 dyn_valve_r [{shapeType=Capsule, length=1.7344, radius=1.4555, center=<1.1905 -0.0002 0>, rotation=<rot 90 0 0>}]
 upperarm_l [{shapeType=Capsule, length=18.4465, radius=3.8575, center=<12.3507 -0.5068 1.6515>, rotation=<rot 0 -93.0514 -86.7975>}]
 upperarm_r [{shapeType=Capsule, length=18.2284, radius=4.8361, center=<-12.0202 0.6639 -0.6001>, rotation=<rot 0 -85.558 -85.2086>}]
}
constraints [
 {bone1Name=dyn_valve_leg, bone2Name=calf_l, swing1LimitDegrees=20, swing2LimitDegrees=20, twistLimitDegrees=5}
 {bone1Name=dyn_valve_l, bone2Name=upperarm_l, swing1LimitDegrees=20, swing2LimitDegrees=20, twistLimitDegrees=5}
 {bone1Name=dyn_valve_r, bone2Name=upperarm_r, swing1LimitDegrees=20, swing2LimitDegrees=20, twistLimitDegrees=5}
]
