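# Animation Blueprint for the female medium WavyMan player body, adding secondary motion on top of a pose copied from its attached parent mesh.
# A RigidBody node simulates the override physics asset in root-bone space, then two Anim Dynamics nodes add damped, linear-spring sway to the boot-string bones.
# Constraint nodes then tie the valve bones to upper-arm and twist bones, weighted on the right and single-target on the left.
# The event graph holds no logic.
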AnimBlueprint /Game/Characters/Player/Female/Medium/Bodies/F_MED_WavyMan/Meshes/F_MED_Wavy_Man_Skeleton_AnimBlueprint
target_skeleton=&/F_MED_Wavy_Man_Skeleton
graphs [
 AnimGraph {nodes [
   n1 = RigidBody(ComponentPose=Local To Component(LocalPose=Copy Pose From Mesh() {bUseAttachedParent=true})) {
    bOverrideWorldGravity=true
    overridePhysicsAsset=&/F_MED_Wavy_Man_Physics
    simulationSpace=BaseBoneSpace
    alphaBoolBlend {blendOption=Linear}
    baseBoneRef {boneName=root}
    componentAppliedLinearAccClamp=<5000 5000 5000>
    componentLinearAccScale=<0.5 0.5 0.5>
    componentLinearVelScale=<0.4 0.4 0.4>
    overrideWorldGravity {z=-980}
   }
   n2 = Anim Dynamics - Bone: dyn_boot_strings_01(ComponentPose=@n1, Alpha=0.55, ExternalForce="0,0,0") {
    bGravityOverrideInSimSpace=true
    bLinearSpring=true
    bOverrideLinearDamping=true
    gravityScale=0.33
    linearDampingOverride=0.995
    linearSpringConstant=150
    simulationSpace=RootRelative
    alphaBoolBlend {blendOption=Linear}
    boundBone {boneName=dyn_boot_strings_01}
    physicsBodyDefinitions [{
      sphereCollisionRadius=0
      boundBone {boneName=dyn_boot_strings_01}
      boxExtents=<10 10 10>
      constraintSetup {angularLimitsMax=<0 0 0>, angularLimitsMin=<0 0 0>, angularTarget=<0 0 0>, linearAxesMax=<4 2 0.02>, linearAxesMin=<-4 -2 -0.02>}
      localJointOffset=<0 0 -5>
     }]
   }
   n3 = Anim Dynamics - Bone: dyn_boot_strings_02(ComponentPose=@n2, Alpha=0.3, ExternalForce="0,0,0") {
    bGravityOverrideInSimSpace=true
    bLinearSpring=true
    bOverrideLinearDamping=true
    gravityScale=0.29
    linearDampingOverride=0.995
    linearSpringConstant=150
    simulationSpace=RootRelative
    alphaBoolBlend {blendOption=Linear}
    boundBone {boneName=dyn_boot_strings_02}
    physicsBodyDefinitions [{
      sphereCollisionRadius=0
      boundBone {boneName=dyn_boot_strings_02}
      boxExtents=<10 10 10>
      constraintSetup {angularLimitsMax=<0 0 0>, angularLimitsMin=<0 0 0>, angularTarget=<0 0 0>, linearAxesMax=<4 2 0.02>, linearAxesMin=<-5 -2 -0.02>}
      localJointOffset=<0 0 -6>
     }]
   }
   n4 = "Constraint - dyn_valve_r (PPP)"!(ComponentPose=@n3, Alpha=1, ConstraintWeights_0=0.052, ConstraintWeights_1:float=0.89, ConstraintWeights_2:float=0.058) -> (Pose) {
    alphaBoolBlend {blendOption=Linear}
    boneToModify {boneName=dyn_valve_r}
    constraintSetup [{targetBone {boneName=upperarm_r}}, {targetBone {boneName=upperarm_twist_01_r}}, {targetBone {boneName=upperarm_twist_02_r}}]
    constraintWeights [1,1,1]
   }
   n5 = "Constraint - dyn_valve_l (P)"(ComponentPose=@n4, Alpha=1, ConstraintWeights_0=1) {
    alphaBoolBlend {blendOption=Linear}
    boneToModify {boneName=dyn_valve_l}
    constraintSetup [{targetBone {boneName=upperarm_twist_01_l}}]
    constraintWeights [1]
   }
   Output Pose(Result=Component To Local(ComponentPose=@n5)) {layerGroup=None}
  ]}
 EventGraph {nodes [
   TryGetPawnOwner()
   Event BlueprintUpdateAnimation()
  ]}
]
variables []
